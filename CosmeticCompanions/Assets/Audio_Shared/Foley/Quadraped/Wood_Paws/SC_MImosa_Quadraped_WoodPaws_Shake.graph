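# Sound cue for the shake foley of a wood-pawed quadruped companion (Mimosa) in the cosmetic companions content.
# A random node picks one of three equally weighted shake wave variations on each play.
# It routes through the pets creature sound class with shared companion foley attenuation and a slight volume boost.
# It shares the quadruped footstep concurrency set, so shakes are limited together with sounds using that set.
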SoundCue /CosmeticCompanions/Assets/Audio_Shared/Foley/Quadraped/Wood_Paws/SC_MImosa_Quadraped_WoodPaws_Shake
root=Random_2
nodes [
 {id=Random_2, children [WavePlayer_0,WavePlayer_1,WavePlayer_2], properties {weights [1,1,1]}}
 {
  id=WavePlayer_0
  properties {soundWaveAssetPtr=/CosmeticCompanions/Assets/Audio_Shared/Foley/Quadraped/Wood_Paws/Wavs/Mimosa_Quadruped_WoodPaws_Shake_01}
 }
 {
  id=WavePlayer_1
  properties {soundWaveAssetPtr=/CosmeticCompanions/Assets/Audio_Shared/Foley/Quadraped/Wood_Paws/Wavs/Mimosa_Quadruped_WoodPaws_Shake_02}
 }
 {
  id=WavePlayer_2
  properties {soundWaveAssetPtr=/CosmeticCompanions/Assets/Audio_Shared/Foley/Quadraped/Wood_Paws/Wavs/Mimosa_Quadruped_WoodPaws_Shake_03}
 }
]
properties {
 attenuationSettings=/Game/Sounds/Attenuation/AS_Companion_Foley
 maxDistance=1200
 soundClassObject=&Fort_Creatures_Pets
 volumeMultiplier=1.2
 concurrencySet [/CosmeticCompanions/Assets/Audio_Shared/Foley/Quadraped/CONC_Mimosa_Quadraped_Footstep]
}
